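# Niagara system for the OatmealSpread emote that fakes motion blur with short-lived ring meshes on the gadget's five ball bones.
# Five spawn groups, each rate-scaled by a System.RingSpawn variable, select the bone and mesh size; particles re-sample bone position every update.
# Meshes spin, grow and fade, tinted by time of day.
# Two near-identical emitters layer the effect: Ball_01 larger with tighter vertical jitter, Ball_02 more spread.
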NiagaraSystem /BRCosmetics/Animation/Game/MainPlayer/Emotes/OatmealSpread/FX/NS_MotionBlur
dependencies {
 usedDataInterfaces [&NiagaraDataInterfaceSkeletalMesh, &NiagaraDataInterfaceCurve]
 usedDynamicInputs [&MakeVector, &RandomRangeFloat]
 usedModules [
  &SystemState
  &self:SystemSpawnScript.SystemScriptSource.SystemScriptGraph.NiagaraNodeAssignment_0.SetVariables_#1
  &self:SystemSpawnScript.SystemScriptSource.SystemScriptGraph.NiagaraNodeAssignment_1.SetVariables_#2
  &self:SystemSpawnScript.SystemScriptSource.SystemScriptGraph.NiagaraNodeAssignment_2.SetVariables_#3
  &self:SystemSpawnScript.SystemScriptSource.SystemScriptGraph.NiagaraNodeAssignment_3.SetVariables_#4
  &self:SystemSpawnScript.SystemScriptSource.SystemScriptGraph.NiagaraNodeAssignment_4.SetVariables_#5
 ]
}
emitters {
 Main_Ring_Ball_01 {
  emitterUpdateScript [
   &EmitterState()
   &SpawnRate(SpawnRate=&Multiply_Float(A=25, B=$System.RingSpawn_01), Use Spawn Probability=false)
   &SpawnRate(SpawnGroup=1, SpawnRate=&Multiply_Float(A=25, B=$System.RingSpawn_02), Use Spawn Probability=false) {module=SpawnRate001}
   &SpawnRate(SpawnGroup=2, SpawnRate=&Multiply_Float(A=25, B=$System.RingSpawn_03), Use Spawn Probability=false) {module=SpawnRate002}
   &SpawnRate(SpawnGroup=3, SpawnRate=&Multiply_Float(A=25, B=$System.RingSpawn_04), Use Spawn Probability=false) {module=SpawnRate003}
   &SpawnRate(SpawnGroup=4, SpawnRate=&Multiply_Float(A=25, B=$System.RingSpawn_05), Use Spawn Probability=false) {module=SpawnRate004}
  ]
  particleSpawnScript [
   &InitializeParticle(
    Color=&RandomRangeLinearColor(Minimum=<rgba 0.388 0.4773 0.599 0.75>, Maximum=<rgba 0.2333 0.3458 0.5052 1>)
    Lifetime=0.25
    Mesh Scale Max=&Multiply_VectorByFloat(
     Vector=<1 0.3 1>
     Float=&MaskFloat(
      Mask Source=0
      Mask Parameter=?
      Mask=0
      Masked Float=0.144
      Passthrough Float=&MaskFloat(
       Mask Source=0
       Mask Parameter=?
       Mask=1
       Masked Float=0.108
       Passthrough Float=&MaskFloat(
        Mask Source=0
        Mask Parameter=?
        Mask=2
        Masked Float=0.072
        Passthrough Float=&MaskFloat(
         Mask Source=0
         Mask Parameter=?
         Mask=3
         Masked Float=0.054
         Passthrough Float=&MaskFloat(Mask Source=0, Mask Parameter=?, Mask=4, Masked Float=0.198, Passthrough Float=1)
        )
       )
      )
     )
    )
    Mesh Scale Min=&Multiply_VectorByFloat(
     Vector=<1 0.3 1>
     Float=&MaskFloat(
      Mask Source=0
      Mask Parameter=?
      Mask=0
      Masked Float=0.112
      Passthrough Float=&MaskFloat(
       Mask Source=0
       Mask Parameter=?
       Mask=1
       Masked Float=0.084
       Passthrough Float=&MaskFloat(
        Mask Source=0
        Mask Parameter=?
        Mask=2
        Masked Float=0.056
        Passthrough Float=&MaskFloat(
         Mask Source=0
         Mask Parameter=?
         Mask=3
         Masked Float=0.042
         Passthrough Float=&MaskFloat(Mask Source=0, Mask Parameter=?, Mask=4, Masked Float=0.154, Passthrough Float=1)
        )
       )
      )
     )
    )
    Mesh Scale Mode=::NewEnumerator2
    Position Mode=::NewEnumerator0
    Write Material Random=true
   )
   &CameraOffset(Camera Offset Amount=15)
   &InitialMeshOrientation(
    Mesh Axis To Orient=<0 1 0>
    Mesh Orientation Mode=::NewEnumerator2
    Orientation Coordinate Space=::World
    Orientation Vector=<0 0 1>
    Rotation=&RandomRangeVector(Minimum=<0 0 0>, Maximum=<0 1 0>)
    Use Rotation Vector=true
   )
  ]
  particleUpdateScript [
   &ParticleState()
   &SkeletalMeshLocation(
    Bone / Socket Index=&MaskIntBySpawnGroup(
     Masked Int=0
     Passthrough Int=&MaskIntBySpawnGroup(
      Masked Int=1
      Passthrough Int=&MaskIntBySpawnGroup(
       Masked Int=2
       Passthrough Int=&MaskIntBySpawnGroup(Masked Int=3, Passthrough Int=4, Spawn Group Mask=3)
       Spawn Group Mask=2
      )
      Spawn Group Mask=1
     )
     Spawn Group Mask=0
    )
    Bone Sampling Mode=::NewEnumerator4
    MaskSpawnGroup=false
    Sampled Position Offset=&MaskVector(
     Mask Source=0
     Mask Parameter=?
     Mask=0
     Masked Vector=&RandomRangeVector(Minimum=<0 0 -5>, Maximum=<0 0 5>)
     Passthrough Vector=&MaskVector(
      Mask Source=0
      Mask Parameter=?
      Mask=1
      Masked Vector=&RandomRangeVector(Minimum=<0 0 -5>, Maximum=<0 0 5>)
      Passthrough Vector=&MaskVector(
       Mask Source=0
       Mask Parameter=?
       Mask=2
       Masked Vector=&RandomRangeVector(Minimum=<0 0 -5>, Maximum=<0 0 5>)
       Passthrough Vector=&MaskVector(
        Mask Source=0
        Mask Parameter=?
        Mask=3
        Masked Vector=&RandomRangeVector(Minimum=<0 0 -2>, Maximum=<0 0 2>)
        Passthrough Vector=&MaskVector(
         Mask Source=0
         Mask Parameter=?
         Mask=4
         Masked Vector=&RandomRangeVector(Minimum=<0 0 -5>, Maximum=<0 0 5>)
         Passthrough Vector=<1 1 1>
        )
       )
      )
     )
    )
    Skeletal Mesh=DataInterface() {propertyValues {
      DefaultMesh=/BRCosmetics/Animation/Props/Gadgets_OatmealSpread/Meshes/OatmealSpread_Gadget
      ExcludeBoneName=root
      OptionalComponentClass=null
      PreviewMesh=&OatmealSpread_Gadget
      SkinningMode=SkinOnTheFly
      SoftSourceActor=null
      SourceMode=Default
      UvSetIndex=0
      WholeMeshLOD=-1
      bExcludeBone=false
      bReadDeformedGeometry=true
      bRequireCurrentFrameData=true
      ComponentTags []
      FilteredBones [JNT_Ball_B,JNT_Ball_C,JNT_Ball_D,JNT_Ball_E,JNT_Ball_A]
      FilteredSockets []
      MeshUserParameter {parameter {name=None, typeDefHandle {registeredTypeIndex=1}, varData []}}
      SamplingRegions []
     }}
   )
   &UpdateMeshOrientation(Rotation Rate=0.5, Rotation Vector=<0 -3 0>) {module=UpdateMeshOrientation002}
   &ScaleMeshSize(Scale Factor=&VectorFromFloat(Value=&FloatFromCurve(FloatCurve=DataInterface() {propertyValues {Curve {keys [<at 0 0.7>,<at 0.45 0.875>,<at 1 1.5>]}}}, Scale Curve=1.25)))
   &DynamicMaterialParameters(
    Index 0 Param 1=&FloatFromCurve(FloatCurve=DataInterface() {propertyValues {Curve {keys [
         {arriveTangent=2.4535, interpMode=RCIM_Cubic, leaveTangent=2.4535, tangentMode=RCTM_User, time=0, value=0}
         {interpMode=RCIM_Cubic, tangentMode=RCTM_User, time=0.5, value=0.5}
         {arriveTangent=-2.5865, tangentMode=RCTM_User, time=1, value=0}
        ]}}})
    Module.Index0Param0Write=true
    Module.Index0Param1Write=false
    Module.Index0Param2Write=false
    Module.Index0Param3Write=false
   )
   &SolveForcesAndVelocity(Clamp Velocity=false, Limit Acceleration=false)
   &ScaleColor(
    Scale Alpha=&FloatFromCurve(
     FloatCurve=DataInterface() {propertyValues {Curve {keys [
         {arriveTangent=6.3511, interpMode=RCIM_Cubic, leaveTangent=6.3511, tangentMode=RCTM_User, time=0, value=0}
         {interpMode=RCIM_Cubic, tangentMode=RCTM_User, time=0.5, value=1}
         {tangentMode=RCTM_User, time=1, value=0}
        ]}}}
     Scale Curve=0.3
    )
    Scale RGB=&MakeVectorFromLinearColorRGB(Color=&FNTimeOfDayColor(
      Night Color=<rgba 0.1253 0.1914 0.5 1>
      Morning Color=<rgba 0.5 0.3383 0.3719 1>
      Day Color=<rgba 0.5 0.5 0.5 1>
      Evening Color=<rgba 0.4474 0.4547 0.4578 1>
     ))
    ScaleA=true
    ScaleRGB=true
   )
  ]
  rendererClasses [&NiagaraMeshRendererProperties]
  renderers [{rendererClass=&NiagaraMeshRendererProperties, rendererIndex=0}]
 }
 Main_Ring_Ball_02 {
  emitterUpdateScript [
   &EmitterState()
   &SpawnRate(SpawnRate=&Multiply_Float(A=25, B=$System.RingSpawn_01), Use Spawn Probability=false)
   &SpawnRate(SpawnGroup=1, SpawnRate=&Multiply_Float(A=25, B=$System.RingSpawn_02), Use Spawn Probability=false) {module=SpawnRate001}
   &SpawnRate(SpawnGroup=2, SpawnRate=&Multiply_Float(A=25, B=$System.RingSpawn_03), Use Spawn Probability=false) {module=SpawnRate002}
   &SpawnRate(SpawnGroup=3, SpawnRate=&Multiply_Float(A=25, B=$System.RingSpawn_04), Use Spawn Probability=false) {module=SpawnRate003}
   &SpawnRate(SpawnGroup=4, SpawnRate=&Multiply_Float(A=25, B=$System.RingSpawn_05), Use Spawn Probability=false) {module=SpawnRate004}
  ]
  particleSpawnScript [
   &InitializeParticle(
    Color=&RandomRangeLinearColor(Minimum=<rgba 0.388 0.4773 0.599 0.75>, Maximum=<rgba 0.2333 0.3458 0.5052 1>)
    Lifetime=0.25
    Mesh Scale Max=&Multiply_VectorByFloat(
     Vector=<1 0.3 1>
     Float=&MaskFloat(
      Mask Source=0
      Mask Parameter=?
      Mask=0
      Masked Float=0.144
      Passthrough Float=&MaskFloat(
       Mask Source=0
       Mask Parameter=?
       Mask=1
       Masked Float=0.108
       Passthrough Float=&MaskFloat(
        Mask Source=0
        Mask Parameter=?
        Mask=2
        Masked Float=0.072
        Passthrough Float=&MaskFloat(
         Mask Source=0
         Mask Parameter=?
         Mask=3
         Masked Float=0.054
         Passthrough Float=&MaskFloat(Mask Source=0, Mask Parameter=?, Mask=4, Masked Float=0.198, Passthrough Float=1)
        )
       )
      )
     )
    )
    Mesh Scale Min=&Multiply_VectorByFloat(
     Vector=<1 0.3 1>
     Float=&MaskFloat(
      Mask Source=0
      Mask Parameter=?
      Mask=0
      Masked Float=0.112
      Passthrough Float=&MaskFloat(
       Mask Source=0
       Mask Parameter=?
       Mask=1
       Masked Float=0.084
       Passthrough Float=&MaskFloat(
        Mask Source=0
        Mask Parameter=?
        Mask=2
        Masked Float=0.056
        Passthrough Float=&MaskFloat(
         Mask Source=0
         Mask Parameter=?
         Mask=3
         Masked Float=0.042
         Passthrough Float=&MaskFloat(Mask Source=0, Mask Parameter=?, Mask=4, Masked Float=0.154, Passthrough Float=1)
        )
       )
      )
     )
    )
    Mesh Scale Mode=::NewEnumerator2
    Position Mode=::NewEnumerator0
    Write Material Random=true
   )
   &CameraOffset(Camera Offset Amount=15)
   &InitialMeshOrientation(
    Mesh Axis To Orient=<0 1 0>
    Mesh Orientation Mode=::NewEnumerator2
    Orientation Coordinate Space=::World
    Orientation Vector=<0 0 1>
    Rotation=&RandomRangeVector(Minimum=<0 0 0>, Maximum=<0 1 0>)
    Use Rotation Vector=true
   )
  ]
  particleUpdateScript [
   &ParticleState()
   &SkeletalMeshLocation(
    Bone / Socket Index=&MaskIntBySpawnGroup(
     Masked Int=0
     Passthrough Int=&MaskIntBySpawnGroup(
      Masked Int=1
      Passthrough Int=&MaskIntBySpawnGroup(
       Masked Int=2
       Passthrough Int=&MaskIntBySpawnGroup(Masked Int=3, Passthrough Int=4, Spawn Group Mask=3)
       Spawn Group Mask=2
      )
      Spawn Group Mask=1
     )
     Spawn Group Mask=0
    )
    Bone Sampling Mode=::NewEnumerator4
    MaskSpawnGroup=false
    Sampled Position Offset=&MaskVector(
     Mask Source=0
     Mask Parameter=?
     Mask=0
     Masked Vector=&RandomRangeVector(Minimum=<0 0 -14>, Maximum=<0 0 14>)
     Passthrough Vector=&MaskVector(
      Mask Source=0
      Mask Parameter=?
      Mask=1
      Masked Vector=&RandomRangeVector(Minimum=<0 0 -8>, Maximum=<0 0 8>)
      Passthrough Vector=&MaskVector(
       Mask Source=0
       Mask Parameter=?
       Mask=2
       Masked Vector=&RandomRangeVector(Minimum=<0 0 -8>, Maximum=<0 0 8>)
       Passthrough Vector=&MaskVector(
        Mask Source=0
        Mask Parameter=?
        Mask=3
        Masked Vector=&RandomRangeVector(Minimum=<0 0 -3>, Maximum=<0 0 3>)
        Passthrough Vector=&MaskVector(
         Mask Source=0
         Mask Parameter=?
         Mask=4
         Masked Vector=&RandomRangeVector(Minimum=<0 0 -14>, Maximum=<0 0 14>)
         Passthrough Vector=<1 1 1>
        )
       )
      )
     )
    )
    Skeletal Mesh=DataInterface() {propertyValues {
      DefaultMesh=&OatmealSpread_Gadget
      ExcludeBoneName=root
      OptionalComponentClass=null
      PreviewMesh=&OatmealSpread_Gadget
      SkinningMode=SkinOnTheFly
      SoftSourceActor=null
      SourceMode=Default
      UvSetIndex=0
      WholeMeshLOD=-1
      bExcludeBone=false
      bReadDeformedGeometry=true
      bRequireCurrentFrameData=true
      ComponentTags []
      FilteredBones [JNT_Ball_B,JNT_Ball_C,JNT_Ball_D,JNT_Ball_E,JNT_Ball_A]
      FilteredSockets []
      MeshUserParameter {parameter {name=None, typeDefHandle {registeredTypeIndex=1}, varData []}}
      SamplingRegions []
     }}
   )
   &UpdateMeshOrientation(Rotation Rate=0.5, Rotation Vector=<0 -3 0>) {module=UpdateMeshOrientation002}
   &ScaleMeshSize(Scale Factor=&VectorFromFloat(Value=&FloatFromCurve(FloatCurve=DataInterface() {propertyValues {Curve {keys [<at 0 0.7>,<at 0.45 0.875>,<at 1 1.5>]}}})))
   &DynamicMaterialParameters(
    Index 0 Param 1=&FloatFromCurve(FloatCurve=DataInterface() {propertyValues {Curve {keys [
         {arriveTangent=2.4535, interpMode=RCIM_Cubic, leaveTangent=2.4535, tangentMode=RCTM_User, time=0, value=0}
         {interpMode=RCIM_Cubic, tangentMode=RCTM_User, time=0.5, value=0.5}
         {arriveTangent=-2.5865, tangentMode=RCTM_User, time=1, value=0}
        ]}}})
    Module.Index0Param0Write=true
    Module.Index0Param1Write=false
    Module.Index0Param2Write=false
    Module.Index0Param3Write=false
   )
   &SolveForcesAndVelocity(Clamp Velocity=false, Limit Acceleration=false)
   &ScaleColor(
    Scale Alpha=&FloatFromCurve(
     FloatCurve=DataInterface() {propertyValues {Curve {keys [
         {arriveTangent=6.3511, interpMode=RCIM_Cubic, leaveTangent=6.3511, tangentMode=RCTM_User, time=0, value=0}
         {interpMode=RCIM_Cubic, tangentMode=RCTM_User, time=0.5, value=1}
         {tangentMode=RCTM_User, time=1, value=0}
        ]}}}
     Scale Curve=0.3
    )
    Scale RGB=&MakeVectorFromLinearColorRGB(Color=&FNTimeOfDayColor(
      Night Color=<rgba 0.1253 0.1914 0.5 1>
      Morning Color=<rgba 0.5 0.3383 0.3719 1>
      Day Color=<rgba 0.5 0.5 0.5 1>
      Evening Color=<rgba 0.4474 0.4547 0.4578 1>
     ))
    ScaleA=true
    ScaleRGB=true
   )
  ]
  rendererClasses [&NiagaraMeshRendererProperties]
  renderers [{rendererClass=&NiagaraMeshRendererProperties, rendererIndex=0}]
 }
}
summary {
 systemName=NS_MotionBlur
 emitters [
  {emitterName=Main_Ring_Ball_01, rendererClasses [&NiagaraMeshRendererProperties]}
  {emitterName=Main_Ring_Ball_02, rendererClasses [&NiagaraMeshRendererProperties]}
 ]
 userVariables [
  User.bIsFrontEnd {type=NiagaraBool}
  User.bIsFrontEndPreview {type=NiagaraBool}
 ]
}
